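# Translucent, unlit, two-sided material for Niagara mesh particles that renders an animated golden-fire tentacle in the Demon cosmetic effects.
# World position offset combines a scrolling sine/cosine spiral, SplineThicken width with bend-rotated tangents, and a length-clamped stretch, driven by particle dynamic parameters.
# Panned fire-trail textures, distorted by opposing normal noise, mix with curve-atlas gradients for color and opacity.
# Grow erodes opacity along the length, finished with depth fade.
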Material /BRCosmetics/Effects/Fort_Effects/GameplayPulginsAssets/Demon/M_Tentacle_Spline
graph {
 properties {
  bAutomaticallySetUsageInEditor=false
  bUsedWithNiagaraMeshParticles=true
  blendMode=BLEND_Translucent
  shadingModel=MSM_Unlit
  twoSided=true
  usedShadingModels=Unlit
 }
 nodes [
  n1 = Multiply(A=Add(A=TextureCoordinate(), B=ParticleRandom()), B=ScalarParameter() {defaultValue=4, parameterName=Param_4})
  n2 = ComponentMask(None=@n1) {g=true}
  n3 = MaterialFunctionCall() {materialFunction=/Engine/Functions/Engine_MaterialFunctions02/WorldPositionOffset/ObjectScale}
  n4 = Time()
  n5 = Multiply(A=Multiply(A=ScalarParameter() {defaultValue=1, parameterName=PanSpeed}, B=Constant() {r=2})) {constB=-1}
  n6 = Sine(None=Add(A=Multiply(A=@n2, B=@n3.Scale X), B=Multiply(A=@n4, B=@n5) {constB=-1})) {period=10}
  n7 = Constant3Vector() {constant=<rgba 0 1.5 1.5 0>}
  n8 = Constant3Vector() {constant=<rgba 0 1.5 -1.5 0>}
  n9 = DynamicParameter() {paramNames [Grow, Emissive, Width, SpiralWPODirection]}
  n10 = Transform(None=LinearInterpolate(A=@n7, B=@n8, Alpha=@n9.SpiralWPODirection)) {transformSourceType=TRANSFORMSOURCE_Local}
  n11 = ComponentMask(None=@n1) {g=true}
  n12 = MaterialFunctionCall() {materialFunction=&ObjectScale}
  n13 = Time()
  n14 = Cosine(None=Add(A=Multiply(A=@n11, B=@n12.Scale X), B=Multiply(A=@n13, B=@n5) {constB=-1})) {period=10}
  n15 = Transform(None=LinearInterpolate(A=@n8, B=@n7, Alpha=@n9.SpiralWPODirection)) {transformSourceType=TRANSFORMSOURCE_Local}
  n16 = Multiply(A=Add(A=Multiply(A=@n6, B=@n10), B=Multiply(A=@n14, B=@n15)), B=ScalarParameter() {defaultValue=4, parameterName=SpiralWPO})
  n17 = MaterialFunctionCall() {materialFunction=&LinearGradient}
  n18 = DynamicParameter() {paramNames [Grow, Emissive, Width, SpiralWPODirection]}
  n19 = Multiply(A=@n18.Width, B=ScalarParameter() {defaultValue=30, parameterName=Width})
  n20 = Constant() {r=1}
  n21 = CurveAtlasRowParameter(CurveTime=MaterialFunctionCall() {materialFunction=&LinearGradient}.VGradient) {
   atlas=/Game/Effects/Fort_Effects/GameplayPulginsAssets/Demon/CA_GoldenFire
   curve=/Game/Effects/Fort_Effects/GameplayPulginsAssets/Demon/Curve_GoldenFire02
   defaultValue=1
   parameterName=Param_3
  }
  n22 = Multiply(A=@n21.A)
  n23 = StaticBool() {value=true}
  n24 = DynamicParameter() {parameterIndex=2}
  n25 = MaterialFunctionCall(Vector=MaterialFunctionCall(X=@n24.Param1, Y=@n24.Param1, Z=@n24.Param1) {materialFunction=&MakeFloat3}.Result) {materialFunction=&SafeNormalize}
  n26 = DynamicParameter() {parameterIndex=3}
  n27 = MaterialFunctionCall(Vector=MaterialFunctionCall(X=@n26.Param1, Y=@n26.Param1, Z=@n26.Param1) {materialFunction=&MakeFloat3}.Result) {materialFunction=&SafeNormalize}
  n28 = MaterialFunctionCall(Vector=CrossProduct(A=@n25.Result, B=@n27.Result)) {materialFunction=&SafeNormalize}
  n29 = TextureCoordinate()
  n30 = Multiply(A=Constant() {r=2}, B=Power(Base=ComponentMask(None=@n29) {g=true}, Exp=ScalarParameter() {defaultValue=3, parameterName=BendPower})) {constA=2}
  n31 = Multiply(A=Multiply(
    A=MaterialFunctionCall!(Normalized V1=@n25.Result, Normalized V2=@n27.Result) {materialFunction=/Engine/Functions/Engine_MaterialFunctions02/Math/AngleBetweenVectors}.0-1
    B=@n30
   ))
  n32 = Constant3Vector()
  n33 = MaterialFunctionCall() {materialFunction=&TangentBasis}
  n34 = Add(A=RotateAboutAxis(NormalizedRotationAxis=@n28.Result, RotationAngle=@n31, PivotPoint=@n32, Position=@n33.X U), B=@n33.X U)
  n35 = Add(A=RotateAboutAxis(NormalizedRotationAxis=@n28.Result, RotationAngle=@n31, PivotPoint=@n32, Position=@n33.Y V), B=@n33.Y V)
  n36 = DynamicParameter() {parameterIndex=1, paramNames [X,Y,Z,MaxLength]}
  n37 = MaterialFunctionCall(X=@n36.X, Y=@n36.X, Z=@n36.X) {materialFunction=&MakeFloat3}
  n38 = MaterialFunctionCall(Vector=@n37.Result) {materialFunction=&SafeNormalize}
  n39 = Clamp(None=Length(None=@n37.Result) {is=class MaterialExpressionLength}, Max=Multiply(A=@n36.MaxLength) {constB=100}) {maxDefault=300}
  Add(
   A=Add(
    A=Multiply(A=@n16, B=@n17.VGradient)
    B=MaterialFunctionCall!(
     TangentU_WS=@n34
     TangentV_WS=@n35
     WidthBase=@n19
     WidthTip=@n20
     WorldPosition
     UVs for Projection
     UVs for Thickness=@n22
     UVs for texturing
     Expand U or V UV Channel=@n23
     DeriveNormalZ
     AdditionalNormal
     AngleCorrectedNormals
     FlattenPixelNormal
     UVs For NormalShading
    ) {materialFunction=/Game/Effects/Niagara/Modules/FollowingDeform/Material/SplineThickenWithTangentInput}.WorldPositionOffset
   ) {constB=0}
   B=LinearInterpolate(B=Multiply(A=@n38.Result, B=@n39), Alpha=@n30)
  ) {constB=0}
  n40 = Add(
   A=NamedRerouteUsage()
   B=MaterialFunctionCall(XY=TextureCoordinate()) {materialFunction=&Swizzle}.YX
  )
  n41 = AppendVector(A=Constant() {r=3}, B=Constant() {r=1})
  n42 = MaterialFunctionCall(UV Channel=MaterialFunctionCall(UVs=Multiply(A=@n40, B=@n41), Texture Scale=Constant2Vector() {g=1.2, r=-1}) {materialFunction=&ScaleUVsByCenter}.UVs) {materialFunction=&LinearGradient}
  n43 = AppendVector(A=OneMinus(None=@n42.UGradient), B=@n42.VGradient)
  n44 = Multiply(A=Time(), B=Constant() {r=2})
  n45 = MaterialFunctionCall(UV Channel=MaterialFunctionCall(UVs=Multiply(A=@n40), Texture Scale=Constant2Vector() {g=1.5, r=-0.7}) {materialFunction=&ScaleUVsByCenter}.UVs) {materialFunction=&LinearGradient}
  n46 = AppendVector(A=OneMinus(None=@n45.UGradient), B=@n45.VGradient)
  n47 = Multiply(
   A=TextureSample(UVs=Panner(Coordinate=@n43, Time=@n44) {speedX=-0.25}) {texture=/Game/Effects/Fort_Effects/Textures/Fire/T_Stylized_Fire_Trail}.R
   B=TextureSample(UVs=Panner(Coordinate=@n46, Time=@n44) {speedX=-0.35}) {texture=&T_Stylized_Fire_Trail}.R
  )
  n48 = Power(Base=@n47, Exp=Constant() {r=1})
  n49 = Power(
   Base=MaterialFunctionCall() {materialFunction=&LinearGradient}.VGradient
   Exp=Constant() {r=0.6}
  ) {constExponent=0.5}
  n50 = Constant() {r=0.5}
  n51 = Saturate(None=Subtract(A=SphereMask(A=@n49, B=@n50) {attenuationRadius=0.5, hardnessPercent=0}, B=ScalarParameter() {defaultValue=0.2}) {constB=0.2})
  n52 = OneMinus(None=Multiply(A=@n51))
  n53 = ParticleColor()
  n54 = Constant() {r=1.1}
  n55 = MaterialFunctionCall() {materialFunction=&ObjectScale}
  n56 = Multiply(A=Add(A=TextureCoordinate(), B=ParticleRandom()), B=AppendVector(A=Constant() {r=0.45}, B=Multiply(A=@n54, B=@n55.Scale X)))
  n57 = Multiply(A=Constant2Vector() {g=-0.3, r=0.1}, B=Multiply(A=ScalarParameter() {defaultValue=1, parameterName=PanSpeed}, B=Constant() {r=1.6}))
  n58 = Sine(None=Divide(A=MaterialFunctionCall() {materialFunction=&LinearGradient}.UGradient))
  n59 = Divide(
   A=TextureSample(UVs=Panner(Coordinate=@n56, Speed=@n57) {speedX=-0.02, speedY=-0.3}) {samplerType=SAMPLERTYPE_Normal, texture=/BRCosmetics/Effects/Fort_Effects/Emotes/UltraLight/Texture/T_DirectionalNoise_Normal}.RGB
   B=LinearInterpolate(A=Constant() {r=2}, B=Constant() {r=7}, Alpha=Saturate(None=@n58)) {constA=1, constB=4}
  )
  n60 = MaterialFunctionCall(UV Channel=Add(A=TextureCoordinate(), B=ComponentMask(None=@n59) {g=true, r=true})) {materialFunction=&LinearGradient}
  n61 = Saturate(None=Multiply(A=@n60.VGradient) {constB=4})
  n62 = Multiply(
   A=MaterialFunctionCall() {materialFunction=&LinearGradient}.VGradient
   B=LinearInterpolate(Alpha=@n18.Grow) {constA=10, constB=0}
  )
  n63 = Subtract(A=Multiply(A=Multiply(A=@n48, B=@n52), B=Multiply(A=@n61, B=@n53.A)), B=@n62)
  n64 = CurveAtlasRowParameter(CurveTime=@n60.UGradient) {
   atlas=&CA_GoldenFire
   curve=/Game/Effects/Fort_Effects/GameplayPulginsAssets/Demon/Curve_GoldenFire01
   parameterName=Param_1
  }
  n65 = Sine(None=Divide(A=MaterialFunctionCall() {materialFunction=&LinearGradient}.UGradient))
  n66 = Constant() {r=5}
  n67 = Constant() {r=100}
  n68 = MaterialFunctionCall(Distance=Constant() {r=2000}) {materialFunction=&Near_Camera_Fade}
  n69 = Saturate(None=Multiply(A=Saturate(None=@n65), B=LinearInterpolate(A=@n66, B=@n67, Alpha=@n68.Result) {constA=2, constB=100}) {constB=2})
  n70 = Saturate(None=Subtract(A=Multiply(A=@n64.A), B=OneMinus(None=@n69)) {constB=0.1})
  n71 = Power(Base=Saturate(None=Subtract(A=@n70, B=@n62) {constB=0.1}), Exp=Constant() {r=2})
  n72 = Multiply(A=Constant() {r=2.5}, B=ParticleColor().A)
  DepthFade(Opacity=Add(A=Saturate(None=@n63), B=Multiply(A=@n71, B=@n72)))
  MaterialFunctionCall() {materialFunction=&DebugTimeSine}
  n73 = CurveAtlasRowParameter(CurveTime=@n47) {
   atlas=&CA_GoldenFire
   curve=/Game/Effects/Fort_Effects/GameplayPulginsAssets/Demon/Curve_GoldenFire03
   defaultValue=2
   parameterName=Param_5
  }
  n74 = Multiply(A=Multiply(A=@n73), B=@n52)
  n75 = ScalarParameter() {defaultValue=5, parameterName=Param_1}
  n76 = CurveAtlasRowParameter(CurveTime=@n60.UGradient) {
   atlas=&CA_GoldenFire
   curve=&Curve_GoldenFire02
   defaultValue=1
   parameterName=Param_2
  }
  n77 = LinearInterpolate(A=LinearInterpolate(A=@n64, B=Divide(A=@n64, B=@n75), Alpha=@n51), B=@n76, Alpha=OneMinus(None=@n61))
  n78 = Multiply(A=Constant() {r=1}, B=@n18.Emissive)
  MaterialFunctionCall(Emissive=Add(A=Multiply(A=@n74, B=@n61), B=Multiply(A=@n77, B=@n78))) {materialFunction=&MF_FX_EmissiveAdjustments}
  n79 = Constant() {r=4}
  n80 = Constant() {r=1.5}
  n81 = Multiply(
   A=MaterialFunctionCall(XY=TextureCoordinate()) {materialFunction=&Swizzle}.YX
   B=Divide(A=AppendVector(A=@n79, B=@n80), B=Constant() {r=1})
  )
  n82 = Constant() {r=0.5}
  n83 = Constant() {r=3}
  n84 = Constant() {r=3}
  n85 = Panner(Coordinate=@n81, Speed=Divide(A=AppendVector(A=@n82, B=@n83), B=@n84)) {speedX=0.2, speedY=0.2}
  n86 = Round(None=MaterialFunctionCall() {materialFunction=&LinearGradient}.UGradient)
  n87 = Multiply(
   A=ComponentMask(None=TextureSample(UVs=@n85) {samplerType=SAMPLERTYPE_Normal, texture=/Game/Global/Textures/Default/Noise/T_Normal_Noise01}.RGB) {g=true, r=true}
   B=OneMinus(None=@n86)
  )
  n88 = Multiply(A=@n83) {constB=-1}
  n89 = Panner(Coordinate=@n81, Speed=Divide(A=AppendVector(A=@n82, B=@n88), B=@n84)) {speedX=0.2, speedY=-0.2}
  n90 = Multiply(
   A=ComponentMask(None=TextureSample(UVs=@n89) {samplerType=SAMPLERTYPE_Normal, texture=&T_Normal_Noise01}.RGB) {g=true, r=true}
   B=@n86
  )
  NamedRerouteDeclaration(None=Divide(A=Add(A=@n87, B=@n90), B=Constant() {r=80})) {name=UVDistortionOpposite, nodeColor=<rgba 0 0.7656 1 1>}
 ]
}
